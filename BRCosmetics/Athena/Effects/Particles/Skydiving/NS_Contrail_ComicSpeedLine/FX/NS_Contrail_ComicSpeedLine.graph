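# Skydiving contrail cosmetic that draws comic-style speed-line sprites around the falling player.
# Hands, feet, shoulder and thigh emitters pin sprites to skeletal mesh bones chosen by spawn group; all but feet stretch with owner speed.
# A GPU surround emitter streams thin upward lines from a cylinder, scaling spawn rate and velocity with speed.
# Randomized dynamic material parameters vary the lines, which fade in and out; body and radial emitters are disabled.
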NiagaraSystem /BRCosmetics/Athena/Effects/Particles/Skydiving/NS_Contrail_ComicSpeedLine/FX/NS_Contrail_ComicSpeedLine
dependencies {
 usedDataInterfaces [&NiagaraDataInterfaceCurve, &NiagaraDataInterfaceColorCurve, &NiagaraDataInterfaceCamera]
 usedDynamicInputs []
 usedModules [&SystemState]
}
emitters {
 SpeedLine_Body {
  bEnabled=false
  emitterSpawnScript [&FN_PreserveSkydiveUserVariables() {module=PreserveUserVariables}]
  emitterUpdateScript [
   &EmitterState(Life Cycle Mode=::NewEnumerator1, Loop Duration=1, Scalability Mode=::NewEnumerator1, UseLoopDelay=false)
   &SpawnBurst_Instantaneous(Spawn Count=2, Use Spawn Probability=false) {enabled=false}
   &SpawnRate(SpawnRate=10, Use Spawn Probability=false) {enabled=false}
   &SpawnRate(
    SpawnRate=&Multiply_Float(
     A=5
     B=&FloatFromCurve(
      FloatCurve=DataInterface() {propertyValues {Curve {keys [{arriveTangent=1.9075, interpMode=RCIM_Cubic, leaveTangent=1.9076, tangentMode=RCTM_User, time=0, value=0.5}, <at 0.1 5 RCIM_Cubic>]}}}
      CurveIndex=&Divide_Float(A=&VectorLength(Vector=$Engine.Owner.Velocity), B=100000)
     )
    )
    Use Spawn Probability=false
   ) {module=SpawnRate001}
  ]
  particleSpawnScript [
   /Game/Effects/Niagara/Modules/Groups/FN_SetSpawnGroup_v1/FN_SetSpawnGroup(Emitter Bone Count=2) {enabled=false}
   /SetVariables(
    Particles.BoneIndexRandom:NiagaraInt32=&UniformRangedInt(Minimum=4, Maximum=22)
    Particles.Dy_Darkness:NiagaraFloat=&UniformRangedFloat(Minimum=20, Maximum=30)
    Particles.Dy_U_Tile:NiagaraFloat=&UniformRangedFloat(Minimum=0.1, Maximum=1)
    Particles.Dy_offsetUV:NiagaraFloat=&UniformRangedFloat(Minimum=3, Maximum=4)
    Particles.FallingSpeed:Vector3f=$Engine.Owner.Velocity
    Particles.PositionRandom:Vector3f=&UniformRangedVector(Minimum=<-5 -5 0>, Maximum=<5 5 0>)
   ) {isSetParameters=true, module=SetVariables_#1}
   /Niagara/Modules/Spawn/Initialization/InitializeParticle(
    Color=<rgba 1 0.1583 0.0084 1>
    Lifetime=&UniformRangedFloat(Minimum=1, Maximum=1.5)
    Mesh Scale=&Multiply_Vector(
     A=<1 1 1>
     B=&VectorFromFloat(Value=&UniformRangedFloat(Minimum=0.2, Maximum=1))
    )
    Module.Write Color=true
    Module.Write Lifetime=true
    Module.Write Mass=true
    Module.Write Material Random=true
    Module.Write Position=false
    Module.Write Scale=false
    Module.Write SpriteRotation=true
    Module.Write SpriteSize=true
    Position=&SimulationPosition()
    Sprite Rotation=90
    Sprite Size=&MakeVector2D(
     X=&Multiply_Float(A=15, B=&Lerp_Float(A=0.3, B=1, Alpha=$Particles.Dy_U_Tile))
     Y=&Multiply_Float(
      A=&UniformRangedFloat(Minimum=50, Maximum=150)
      B=&FloatFromCurve(
       FloatCurve=DataInterface() {propertyValues {Curve {keys [<at 0 1>,<at 1 2.5>]}}}
       CurveIndex=&Divide_Float(A=&VectorLength(Vector=$Particles.FallingSpeed), B=9000)
      )
     )
    )
   )
   &CylinderLocation(
    Cylinder Height=1
    Lathe Profile=DataInterface() {propertyValues {Curve {keys []}}}
    MaskSpawnGroup=false
    Module.Enable Lathe Profile=false
    Module.Override Randomness=false
    Module.Surface Only=false
   ) {enabled=false}
   &SampleSkeletalMeshSkeleton(Mesh:NiagaraDataInterfaceSkeletalMesh=$User.SkeletalMesh, Specific Bone Index:NiagaraInt32=$Particles.SpawnGroup) {enabled=false}
   /Niagara/Modules/Spawn/Location/SkeletalMeshSkeletonLocation(Module.Inherit Velocity=false, Module.Use Sampled Bone as Ribbon ID=true) {enabled=false}
   &InheritVelocity(Module.Limit Speed=true, Velocity Limit=0.1, Velocity Scale=<-1 -1 -1>) {enabled=false}
   &AddVelocity(Velocity=<0 0 1>)
   /Niagara/Modules/Spawn/Orientation/V2/InitialMeshOrientation(
    Orientation Vector=&UniformRangedVector(Minimum=<-1 -1 -1>, Maximum=<1 1 1>)
    Use Orientation Vector=true
    Use Rotation Vector=false
   ) {enabled=false}
   /SetVariables(Particles.Color=&MakeLinearColorFromVectorAndFloat(
     "Vector (RGB)"=&MakeVectorFromLinearColorRGB(Color=&ColorFromCurve(ColorCurve=DataInterface() {propertyValues {
         AlphaCurve {keys [<at -0.0023 1 RCIM_Cubic>,<at 0.3756 1>,<at 0.9977 1 RCIM_Cubic>]}
         BlueCurve {keys [<at 0.3756 0.0782>,<at 0.4415 0.0786 RCIM_Cubic>]}
         GreenCurve {keys [<at 0.3756 0>,<at 0.4415 0 RCIM_Cubic>]}
         RedCurve {keys [<at 0.3756 1>,<at 0.4415 1 RCIM_Cubic>]}
        }}))
     "Float (Alpha)"=1
    )) {enabled=false, isSetParameters=true, module=SetVariables_#2}
   &InheritVelocity(Module.Limit Speed=false, Velocity Scale=<1 1 1.05>) {enabled=false, module=InheritVelocity001}
   /Niagara/Modules/Spawn/Location/SphereLocation(MaskSpawnGroup=false, Module.Override Randomness=false, Non Uniform Scale=<1.5 1.5 1.5>, Sphere Radius=20, Surface Only=false) {enabled=false}
   &DynamicMaterialParameters(
    Index 0 Param 1=$Particles.Dy_offsetUV
    Index 0 Param 2=0.2
    Index 0 Param 3=$Particles.Dy_Darkness
    Index 1 Param 1=0.01
    Index 1 Param 2=0.1
    Module.Param0WriteEnabled=true
    Module.Param0WriteEnabled1=true
    Module.Param1WriteEnabled=true
    Module.Param1WriteEnabled1=true
    Module.Param2WriteEnabled=true
    Module.Param2WriteEnabled1=true
    Module.Param3WriteEnabled=true
    Module.Param3WriteEnabled1=true
    Write Parameter Index 1=true
   )
  ]
  particleUpdateScript [
   &SkeletalMeshLocation(
    Bone / Socket Index=$Particles.BoneIndexRandom
    Bone Sampling Mode=::NewEnumerator4
    Inherit Velocity=false
    MaskSpawnGroup=false
    Sampled Position Offset=$Particles.PositionRandom
    Skeletal Mesh=$User.SkeletalMesh
   )
   &ParticleState()
   &Drag(Drag=0.5, Use Linear Drag=true, Use Rotational Drag=false) {enabled=false}
   /Niagara/Modules/Update/Forces/CurlNoiseForce(
    Noise Frequency:NiagaraFloat=&UniformRangedFloat(Minimum=0.1, Maximum=1)
    Noise Strength=&UniformRangedFloat(Minimum=300, Maximum=600)
    Pan Noise=false
    Pan Noise Field=<1 1 0>
   ) {enabled=false}
   &SolveForcesAndVelocity(Clamp Velocity=false, Limit Acceleration=false)
   &ScaleSpriteSize(Scale Factor=&Multiply_Vector2(
     A=<1 1>
     B=&Vector2DFromFloat(Value=&FloatFromCurve(FloatCurve=DataInterface() {propertyValues {Curve {keys [
           <at 0 0.15 RCIM_Cubic>
           {arriveTangent=0.0667, leaveTangent=0.0667, time=0.1259, value=1}
           {arriveTangent=-1.1418, interpMode=RCIM_Cubic, leaveTangent=-1.1418, time=0.8397, value=1}
           {interpMode=RCIM_Cubic, tangentMode=RCTM_User, time=1, value=0.002}
          ]}}}))
    )) {enabled=false}
   &ScaleSpriteSize(Scale Factor=&Multiply_Vector2(
     A=<1 1>
     B=&Vector2DFromFloat(Value=&FloatFromCurve(
       FloatCurve=DataInterface() {propertyValues {Curve {keys [<at 0.4336 1.0007>,<at 0.4757 0.8507>,<at 0.5438 1.5807>,<at 0.6167 0.9007>,<at 0.6539 1.0007>,<at 1 1 RCIM_Cubic>]}}}
       Scale Curve=1.2
      ))
    )) {enabled=false, module=ScaleSpriteSize001}
   &ScaleColor(
    Scale Alpha=&FloatFromCurve(FloatCurve=DataInterface() {propertyValues {Curve {keys [<at 0 0 RCIM_Cubic>,<at 0.3634 1.0074 RCIM_Cubic>,<at 1 0 RCIM_Cubic>]}}})
    ScaleA=true
    ScaleRGB=true
   )
   &CameraOffset(Camera Offset Amount=10) {enabled=false}
  ]
  rendererClasses [&NiagaraSpriteRendererProperties]
  renderers [{rendererClass=&NiagaraSpriteRendererProperties}]
 }
 SpeedLine_Feet {
  emitterSpawnScript [&FN_PreserveSkydiveUserVariables() {module=PreserveUserVariables}]
  emitterUpdateScript [
   &EmitterState(Life Cycle Mode=::NewEnumerator1, Loop Duration=1, Scalability Mode=::NewEnumerator1, UseLoopDelay=false)
   &SpawnBurst_Instantaneous(Spawn Count=2, Use Spawn Probability=false) {enabled=false}
   &SpawnRate(SpawnGroup=2, SpawnRate=2, Use Spawn Probability=false)
   &SpawnRate(SpawnGroup=3, SpawnRate=2, Use Spawn Probability=false) {module=SpawnRate001}
  ]
  particleSpawnScript [
   &FN_SetSpawnGroup(Emitter Bone Count=2) {enabled=false}
   /SetVariables(
    Particles.Dy_Darkness:NiagaraFloat=&UniformRangedFloat(Minimum=10, Maximum=30)
    Particles.Dy_U_Tile:NiagaraFloat=&UniformRangedFloat(Minimum=0.1, Maximum=1)
    Particles.Dy_offsetUV:NiagaraFloat=&UniformRangedFloat(Minimum=2, Maximum=3)
    Particles.PositionRandom:Vector3f=&UniformRangedVector(Minimum=<-5 -5 0>, Maximum=<5 5 0>)
   ) {isSetParameters=true, module=SetVariables_#1}
   &InitializeParticle(
    Color=<rgba 1 0.1583 0.0084 1>
    Lifetime=&UniformRangedFloat(Minimum=1, Maximum=1.5)
    Mesh Scale=&Multiply_Vector(
     A=<1 1 1>
     B=&VectorFromFloat(Value=&UniformRangedFloat(Minimum=0.2, Maximum=1))
    )
    Module.Write Color=true
    Module.Write Lifetime=true
    Module.Write Mass=true
    Module.Write Material Random=true
    Module.Write Position=false
    Module.Write Scale=false
    Module.Write SpriteRotation=true
    Module.Write SpriteSize=true
    Position=&SimulationPosition()
    Sprite Rotation=90
    Sprite Size=&MakeVector2D(
     X=&Multiply_Float(A=15, B=&Lerp_Float(A=0.3, B=1, Alpha=$Particles.Dy_U_Tile))
     Y=&UniformRangedFloat(Minimum=100, Maximum=200)
    )
   )
   &CylinderLocation(
    Cylinder Height=1
    Lathe Profile=DataInterface() {propertyValues {Curve {keys []}}}
    MaskSpawnGroup=false
    Module.Enable Lathe Profile=false
    Module.Override Randomness=false
    Module.Surface Only=false
   ) {enabled=false}
   &SampleSkeletalMeshSkeleton(Mesh:NiagaraDataInterfaceSkeletalMesh=$User.SkeletalMesh, Specific Bone Index:NiagaraInt32=$Particles.SpawnGroup) {enabled=false}
   &SkeletalMeshSkeletonLocation(Module.Inherit Velocity=false, Module.Use Sampled Bone as Ribbon ID=true) {enabled=false}
   &InheritVelocity(Module.Limit Speed=true, Velocity Limit=0.1, Velocity Scale=<-1 -1 -1>) {enabled=false}
   &AddVelocity(Velocity=<0 0 1>)
   &InitialMeshOrientation(
    Orientation Vector=&UniformRangedVector(Minimum=<-1 -1 -1>, Maximum=<1 1 1>)
    Use Orientation Vector=true
    Use Rotation Vector=false
   ) {enabled=false}
   /SetVariables(Particles.Color=&MakeLinearColorFromVectorAndFloat(
     "Vector (RGB)"=&MakeVectorFromLinearColorRGB(Color=&ColorFromCurve(ColorCurve=DataInterface() {propertyValues {
         AlphaCurve {keys [<at -0.0023 1 RCIM_Cubic>,<at 0.3756 1>,<at 0.9977 1 RCIM_Cubic>]}
         BlueCurve {keys [<at 0.3756 0.0782>,<at 0.4415 0.0786 RCIM_Cubic>]}
         GreenCurve {keys [<at 0.3756 0>,<at 0.4415 0 RCIM_Cubic>]}
         RedCurve {keys [<at 0.3756 1>,<at 0.4415 1 RCIM_Cubic>]}
        }}))
     "Float (Alpha)"=1
    )) {enabled=false, isSetParameters=true, module=SetVariables_#2}
   &InheritVelocity(Module.Limit Speed=false, Velocity Scale=<1 1 1.05>) {enabled=false, module=InheritVelocity001}
   &SphereLocation(MaskSpawnGroup=false, Module.Override Randomness=false, Non Uniform Scale=<1.5 1.5 1.5>, Sphere Radius=20, Surface Only=false) {enabled=false}
   &DynamicMaterialParameters(
    Index 0 Param 1=$Particles.Dy_offsetUV
    Index 0 Param 2=$Particles.Dy_U_Tile
    Index 0 Param 3=$Particles.Dy_Darkness
    Index 0 Param 4=0.5
    Module.Param0WriteEnabled=true
    Module.Param1WriteEnabled=true
    Module.Param2WriteEnabled=true
    Module.Param3WriteEnabled=true
   )
   &CameraOffset(Camera Offset Amount=&UniformRangedFloat(Minimum=20, Maximum=20)) {enabled=false}
  ]
  particleUpdateScript [
   &SkeletalMeshLocation(
    Bone / Socket Index=&SpawnGroup()
    Bone Sampling Mode=::NewEnumerator4
    Inherit Velocity=false
    MaskSpawnGroup=false
    Sampled Position Offset=$Particles.PositionRandom
    Skeletal Mesh=$User.SkeletalMesh
   )
   &ParticleState()
   &Drag(Drag=0.5, Use Linear Drag=true, Use Rotational Drag=false) {enabled=false}
   &CurlNoiseForce(
    Noise Frequency:NiagaraFloat=&UniformRangedFloat(Minimum=0.1, Maximum=1)
    Noise Strength=&UniformRangedFloat(Minimum=300, Maximum=600)
    Pan Noise=false
    Pan Noise Field=<1 1 0>
   ) {enabled=false}
   &SolveForcesAndVelocity(Clamp Velocity=false, Limit Acceleration=false)
   &ScaleSpriteSize(Scale Factor=&Multiply_Vector2(
     A=<1 1>
     B=&Vector2DFromFloat(Value=&FloatFromCurve(FloatCurve=DataInterface() {propertyValues {Curve {keys [
           <at 0 0.15 RCIM_Cubic>
           {arriveTangent=0.0667, leaveTangent=0.0667, time=0.1259, value=1}
           {arriveTangent=-1.1418, interpMode=RCIM_Cubic, leaveTangent=-1.1418, time=0.8397, value=1}
           {interpMode=RCIM_Cubic, tangentMode=RCTM_User, time=1, value=0.002}
          ]}}}))
    )) {enabled=false}
   &ScaleSpriteSize(Scale Factor=&Multiply_Vector2(
     A=<1 1>
     B=&Vector2DFromFloat(Value=&FloatFromCurve(
       FloatCurve=DataInterface() {propertyValues {Curve {keys [<at 0.4336 1.0007>,<at 0.4757 0.8507>,<at 0.5438 1.5807>,<at 0.6167 0.9007>,<at 0.6539 1.0007>,<at 1 1 RCIM_Cubic>]}}}
       Scale Curve=1.2
      ))
    )) {enabled=false, module=ScaleSpriteSize001}
   &ScaleColor(
    Scale Alpha=&FloatFromCurve(FloatCurve=DataInterface() {propertyValues {Curve {keys [<at 0 0 RCIM_Cubic>,<at 0.3634 1.0074 RCIM_Cubic>,<at 1 0 RCIM_Cubic>]}}})
    ScaleA=true
    ScaleRGB=true
   )
   &CameraOffset(Camera Offset Amount=10) {module=CameraOffset001}
  ]
  rendererClasses [&NiagaraSpriteRendererProperties]
  renderers [{rendererClass=&NiagaraSpriteRendererProperties}]
 }
 SpeedLine_Hands {
  emitterSpawnScript [&FN_PreserveSkydiveUserVariables() {module=PreserveUserVariables}]
  emitterUpdateScript [
   &EmitterState(Life Cycle Mode=::NewEnumerator1, Loop Duration=3, Scalability Mode=::NewEnumerator1, UseLoopDelay=false)
   &SpawnBurst_Instantaneous(Spawn Count=2, Use Spawn Probability=false) {enabled=false}
   &SpawnRate(SpawnRate=2, Use Spawn Probability=false)
   &SpawnRate(SpawnGroup=1, SpawnRate=2, Use Spawn Probability=false) {module=SpawnRate001}
  ]
  particleSpawnScript [
   &FN_SetSpawnGroup(Emitter Bone Count=2) {enabled=false}
   /SetVariables(
    Particles.Dy_Darkness:NiagaraFloat=&UniformRangedFloat(Minimum=30, Maximum=1000)
    Particles.Dy_U_Tile:NiagaraFloat=&UniformRangedFloat(Minimum=0.5, Maximum=1)
    Particles.Dy_offsetUV:NiagaraFloat=&UniformRangedFloat(Minimum=0, Maximum=1)
    Particles.FallingSpeed:Vector3f=$Engine.Owner.Velocity
    Particles.PositionRandom:Vector3f=&UniformRangedVector(Minimum=<-2.5 -2.5 0>, Maximum=<2.5 2.5 0>)
   ) {isSetParameters=true, module=SetVariables_#1}
   &InitializeParticle(
    Color=<rgba 1 0.1583 0.0084 1>
    Lifetime=&UniformRangedFloat(Minimum=1, Maximum=1)
    Mesh Scale=&Multiply_Vector(
     A=<1 1 1>
     B=&VectorFromFloat(Value=&UniformRangedFloat(Minimum=0.2, Maximum=1))
    )
    Module.Write Color=true
    Module.Write Lifetime=true
    Module.Write Mass=true
    Module.Write Material Random=true
    Module.Write Position=false
    Module.Write Scale=false
    Module.Write SpriteRotation=true
    Module.Write SpriteSize=true
    Position=&SimulationPosition()
    Sprite Rotation=90
    Sprite Size=&MakeVector2D(
     X=&Multiply_Float(A=20, B=&Lerp_Float(A=0.3, B=1, Alpha=$Particles.Dy_U_Tile))
     Y=&Multiply_Float(
      A=&UniformRangedFloat(Minimum=200, Maximum=250)
      B=&FloatFromCurve(
       FloatCurve=DataInterface() {propertyValues {Curve {keys [<at 0 1>,<at 1 2.5>]}}}
       CurveIndex=&Divide_Float(A=&VectorLength(Vector=$Particles.FallingSpeed), B=9000)
      )
     )
    )
   )
   &CylinderLocation(
    Cylinder Height=1
    Lathe Profile=DataInterface() {propertyValues {Curve {keys []}}}
    MaskSpawnGroup=false
    Module.Enable Lathe Profile=false
    Module.Override Randomness=false
    Module.Surface Only=false
   ) {enabled=false}
   &SampleSkeletalMeshSkeleton(Mesh:NiagaraDataInterfaceSkeletalMesh=$User.SkeletalMesh, Specific Bone Index:NiagaraInt32=$Particles.SpawnGroup) {enabled=false}
   &SkeletalMeshSkeletonLocation(Module.Inherit Velocity=false, Module.Use Sampled Bone as Ribbon ID=true) {enabled=false}
   &InheritVelocity(Module.Limit Speed=true, Velocity Limit=0.1, Velocity Scale=<-1 -1 -1>) {enabled=false}
   &AddVelocity(Velocity=<0 0 1>)
   &InitialMeshOrientation(
    Orientation Vector=&UniformRangedVector(Minimum=<-1 -1 -1>, Maximum=<1 1 1>)
    Use Orientation Vector=true
    Use Rotation Vector=false
   ) {enabled=false}
   /SetVariables(Particles.Color=&MakeLinearColorFromVectorAndFloat(
     "Vector (RGB)"=&MakeVectorFromLinearColorRGB(Color=&ColorFromCurve(ColorCurve=DataInterface() {propertyValues {
         AlphaCurve {keys [<at -0.0023 1 RCIM_Cubic>,<at 0.3756 1>,<at 0.9977 1 RCIM_Cubic>]}
         BlueCurve {keys [<at 0.3756 0.0782>,<at 0.4415 0.0786 RCIM_Cubic>]}
         GreenCurve {keys [<at 0.3756 0>,<at 0.4415 0 RCIM_Cubic>]}
         RedCurve {keys [<at 0.3756 1>,<at 0.4415 1 RCIM_Cubic>]}
        }}))
     "Float (Alpha)"=1
    )) {enabled=false, isSetParameters=true, module=SetVariables_#2}
   &InheritVelocity(Module.Limit Speed=false, Velocity Scale=<1 1 1.05>) {enabled=false, module=InheritVelocity001}
   &SphereLocation(MaskSpawnGroup=false, Module.Override Randomness=false, Non Uniform Scale=<1.5 1.5 1.5>, Sphere Radius=20, Surface Only=false) {enabled=false}
   &DynamicMaterialParameters(
    Index 0 Param 1=$Particles.Dy_offsetUV
    Index 0 Param 2=$Particles.Dy_U_Tile
    Index 0 Param 3=$Particles.Dy_Darkness
    Index 0 Param 4=1
    Module.Param0WriteEnabled=true
    Module.Param1WriteEnabled=true
    Module.Param2WriteEnabled=true
    Module.Param3WriteEnabled=true
   )
   &CameraOffset(Camera Offset Amount=&UniformRangedFloat(Minimum=-5, Maximum=5))
  ]
  particleUpdateScript [
   &SkeletalMeshLocation(
    Bone / Socket Index=&SpawnGroup()
    Bone Sampling Mode=::NewEnumerator4
    Inherit Velocity=false
    MaskSpawnGroup=false
    Skeletal Mesh=$User.SkeletalMesh
   )
   &ParticleState()
   &Drag(Drag=0.5, Use Linear Drag=true, Use Rotational Drag=false) {enabled=false}
   &CurlNoiseForce(
    Noise Frequency:NiagaraFloat=&UniformRangedFloat(Minimum=0.1, Maximum=1)
    Noise Strength=&UniformRangedFloat(Minimum=300, Maximum=600)
    Pan Noise=false
    Pan Noise Field=<1 1 0>
   ) {enabled=false}
   &SolveForcesAndVelocity(Clamp Velocity=false, Limit Acceleration=false)
   &ScaleSpriteSize(Scale Factor=&Multiply_Vector2(
     A=<1 1>
     B=&Vector2DFromFloat(Value=&FloatFromCurve(FloatCurve=DataInterface() {propertyValues {Curve {keys [
           <at 0 0.15 RCIM_Cubic>
           {arriveTangent=0.0667, leaveTangent=0.0667, time=0.1259, value=1}
           {arriveTangent=-1.1418, interpMode=RCIM_Cubic, leaveTangent=-1.1418, time=0.8397, value=1}
           {interpMode=RCIM_Cubic, tangentMode=RCTM_User, time=1, value=0.002}
          ]}}}))
    )) {enabled=false}
   &ScaleSpriteSize(Scale Factor=&Multiply_Vector2(
     A=<1 1>
     B=&Vector2DFromFloat(Value=&FloatFromCurve(
       FloatCurve=DataInterface() {propertyValues {Curve {keys [<at 0.4336 1.0007>,<at 0.4757 0.8507>,<at 0.5438 1.5807>,<at 0.6167 0.9007>,<at 0.6539 1.0007>,<at 1 1 RCIM_Cubic>]}}}
       Scale Curve=1.2
      ))
    )) {enabled=false, module=ScaleSpriteSize001}
   &ScaleColor(
    Scale Alpha=&FloatFromCurve(FloatCurve=DataInterface() {propertyValues {Curve {keys [<at 0 0 RCIM_Cubic>,<at 0.3634 1.0074 RCIM_Cubic>,<at 1 0 RCIM_Cubic>]}}})
    ScaleA=true
    ScaleRGB=true
   )
  ]
  rendererClasses [&NiagaraSpriteRendererProperties]
  renderers [{rendererClass=&NiagaraSpriteRendererProperties}]
 }
 SpeedLine_Radial {
  bEnabled=false
  emitterUpdateScript [
   &EmitterState(Life Cycle Mode=::NewEnumerator1, Loop Duration=1, UseLoopDelay=false)
   &SpawnRate(SpawnRate=2, Use Spawn Probability=false)
  ]
  particleSpawnScript [
   &InitializeParticle(
    Lifetime=0.5
    Module.Write Color=true
    Module.Write Lifetime=true
    Module.Write Mass=true
    Module.Write Material Random=true
    Module.Write Position=false
    Module.Write Scale=false
    Module.Write SpriteRotation=false
    Module.Write SpriteSize=true
    Position=&SimulationPosition()
    Sprite Size=<50 30>
   )
   /SetVariables(Particles.CusInitPosition:Vector3f=&Add_Vector(A=Linked() {linkedVariable {name=Particles.Position, type=NiagaraPosition}}, B=<0 70 120>)) {isSetParameters=true}
   &AddVelocity(Velocity=<0 0 10>) {enabled=false}
  ]
  particleUpdateScript [
   &ParticleState()
   &SolveForcesAndVelocity(Clamp Velocity=false, Limit Acceleration=false) {enabled=false}
   &SystemLocation(
    MaskSpawnGroup=false
    Offset=&Add_Vector(
     A=$Particles.CusInitPosition
     B=&Add_Vector(
      A=&Subtract_Vector(
       A=&ReturnCameraVector(Camera Property=::NewEnumerator0, Coordinate Space=::Simulation)
       B=$Particles.CusInitPosition
      )
      B=&Multiply_Vector(
       A=&ReturnCameraVector(
        Camera Property=::NewEnumerator4
        Coordinate Space=::Simulation
        Origin Position=Linked() {linkedVariable {name=Particles.CusInitPosition, type=Vector3f}}
       )
       B=&VectorFromFloat(Value=-20)
      )
     )
    )
    Offset Coordinate Space=::Simulation
   )
   &ScaleColor(ScaleA=true, ScaleRGB=true)
  ]
  rendererClasses [&NiagaraSpriteRendererProperties]
  renderers [{rendererClass=&NiagaraSpriteRendererProperties}]
 }
 SpeedLine_Shoulder {
  emitterSpawnScript [&FN_PreserveSkydiveUserVariables() {module=PreserveUserVariables}]
  emitterUpdateScript [
   &EmitterState(Life Cycle Mode=::NewEnumerator1, Loop Duration=3, Scalability Mode=::NewEnumerator1, UseLoopDelay=false)
   &SpawnBurst_Instantaneous(Spawn Count=2, Use Spawn Probability=false) {enabled=false}
   &SpawnRate(
    SpawnGroup:NiagaraInt32=&UniformRangedInt(Minimum=18, Maximum=19)
    SpawnRate=5
    Use Spawn Probability=false
   )
  ]
  particleSpawnScript [
   &FN_SetSpawnGroup(Emitter Bone Count=2) {enabled=false}
   /SetVariables(
    Particles.Dy_Darkness:NiagaraFloat=&UniformRangedFloat(Minimum=10, Maximum=50)
    Particles.Dy_U_Tile:NiagaraFloat=&UniformRangedFloat(Minimum=0.5, Maximum=1)
    Particles.Dy_offsetUV:NiagaraFloat=&UniformRangedFloat(Minimum=1, Maximum=2)
    Particles.FallingSpeed:Vector3f=$Engine.Owner.Velocity
    Particles.PositionRandom:Vector3f=&UniformRangedVector(Minimum=<-5 -10 2>, Maximum=<10 10 5>)
   ) {isSetParameters=true, module=SetVariables_#1}
   &InitializeParticle(
    Color=<rgba 1 0.1583 0.0084 1>
    Lifetime=&UniformRangedFloat(Minimum=1, Maximum=1)
    Mesh Scale=&Multiply_Vector(
     A=<1 1 1>
     B=&VectorFromFloat(Value=&UniformRangedFloat(Minimum=0.2, Maximum=1))
    )
    Module.Write Color=true
    Module.Write Lifetime=true
    Module.Write Mass=true
    Module.Write Material Random=true
    Module.Write Position=false
    Module.Write Scale=false
    Module.Write SpriteRotation=true
    Module.Write SpriteSize=true
    Position=&SimulationPosition()
    Sprite Rotation=90
    Sprite Size=&MakeVector2D(
     X=&Multiply_Float(A=15, B=&Lerp_Float(A=0.3, B=2, Alpha=$Particles.Dy_U_Tile))
     Y=&Multiply_Float(
      A=&UniformRangedFloat(Minimum=60, Maximum=150)
      B=&FloatFromCurve(
       FloatCurve=DataInterface() {propertyValues {Curve {keys [<at 0 1 RCIM_Cubic>, {arriveTangent=10.2223, interpMode=RCIM_Cubic, leaveTangent=10.2222, tangentMode=RCTM_User, time=1, value=4}]}}}
       CurveIndex=&Divide_Float(A=&VectorLength(Vector=$Particles.FallingSpeed), B=6000)
      )
     )
    )
   )
   &SampleSkeletalMeshSkeleton(Mesh:NiagaraDataInterfaceSkeletalMesh=$User.SkeletalMesh, Specific Bone Index:NiagaraInt32=$Particles.SpawnGroup) {enabled=false}
   &SkeletalMeshSkeletonLocation(Module.Inherit Velocity=false, Module.Use Sampled Bone as Ribbon ID=true) {enabled=false}
   &AddVelocity(Velocity=<0 0 1>)
   &DynamicMaterialParameters(
    Index 0 Param 1=$Particles.Dy_offsetUV
    Index 0 Param 2=$Particles.Dy_U_Tile
    Index 0 Param 3=$Particles.Dy_Darkness
    Index 0 Param 4=&UniformRangedFloat(Minimum=0, Maximum=0.1)
    Module.Param0WriteEnabled=true
    Module.Param1WriteEnabled=true
    Module.Param2WriteEnabled=true
    Module.Param3WriteEnabled=true
   )
   &CameraOffset(Camera Offset Amount=&UniformRangedFloat(Minimum=5, Maximum=10))
  ]
  particleUpdateScript [
   &SkeletalMeshLocation(
    Bone / Socket Index=&SpawnGroup()
    Bone Sampling Mode=::NewEnumerator4
    Inherit Velocity=false
    MaskSpawnGroup=false
    Sampled Position Offset=$Particles.PositionRandom
    Skeletal Mesh=$User.SkeletalMesh
   )
   &ParticleState()
   &SolveForcesAndVelocity(Clamp Velocity=false, Limit Acceleration=false)
   &ScaleColor(
    Scale Alpha=&FloatFromCurve(FloatCurve=DataInterface() {propertyValues {Curve {keys [<at 0 0 RCIM_Cubic>,<at 0.3634 1.0074 RCIM_Cubic>,<at 1 0 RCIM_Cubic>]}}})
    ScaleA=true
    ScaleRGB=true
   )
  ]
  rendererClasses [&NiagaraSpriteRendererProperties]
  renderers [{rendererClass=&NiagaraSpriteRendererProperties}]
 }
 SpeedLine_Surround {
  simTarget=GPUComputeSim
  emitterSpawnScript [&FN_PreserveSkydiveUserVariables() {module=PreserveUserVariables}]
  emitterUpdateScript [
   &EmitterState(Life Cycle Mode=::NewEnumerator1, Loop Duration=1, Scalability Mode=::NewEnumerator1, UseLoopDelay=false)
   &SpawnBurst_Instantaneous(Spawn Count=2, Use Spawn Probability=false) {enabled=false}
   &SpawnRate(
    SpawnRate=&Multiply_Float(
     A=100
     B=&FloatFromCurve(
      FloatCurve=DataInterface() {propertyValues {Curve {keys [{arriveTangent=1.9075, interpMode=RCIM_Cubic, leaveTangent=1.9076, tangentMode=RCTM_User, time=0, value=0.5}, <at 0.1 10 RCIM_Cubic>]}}}
      CurveIndex=&Divide_Float(A=&VectorLength(Vector=$Engine.Owner.Velocity), B=100000)
     )
    )
    Use Spawn Probability=false
   )
  ]
  particleSpawnScript [
   &FN_SetSpawnGroup(Emitter Bone Count=2) {enabled=false}
   /SetVariables(
    Particles.Dy_Darkness:NiagaraFloat=&UniformRangedFloat(Minimum=20, Maximum=20)
    Particles.Dy_U_Tile:NiagaraFloat=&UniformRangedFloat(Minimum=0.1, Maximum=0.1)
    Particles.Dy_offsetUV:NiagaraFloat=&UniformRangedFloat(Minimum=4, Maximum=5)
    Particles.FallingSpeed:Vector3f=$Engine.Owner.Velocity
   ) {isSetParameters=true, module=SetVariables_#1}
   &InitializeParticle(
    Color=<rgba 1 0.1583 0.0084 1>
    Lifetime=&UniformRangedFloat(Minimum=0.2, Maximum=0.2)
    Mesh Scale=&Multiply_Vector(
     A=<1 1 1>
     B=&VectorFromFloat(Value=&UniformRangedFloat(Minimum=0.2, Maximum=1))
    )
    Module.Write Color=true
    Module.Write Lifetime=true
    Module.Write Mass=true
    Module.Write Material Random=true
    Module.Write Position=false
    Module.Write Scale=false
    Module.Write SpriteRotation=true
    Module.Write SpriteSize=true
    Position=&SimulationPosition()
    Sprite Rotation=90
    Sprite Size=&MakeVector2D(
     X=&UniformRangedFloat(Minimum=1, Maximum=3)
     Y=&UniformRangedFloat(Minimum=200, Maximum=300)
    )
   )
   &CylinderLocation(
    Cylinder Height=2
    MaskSpawnGroup=false
    Module.Enable Lathe Profile=false
    Module.Override Randomness=false
    Module.Surface Only=true
    Surface Only Band Thickness=1
   )
   &CylinderLocation(
    Cylinder Height=1
    Lathe Profile=DataInterface() {propertyValues {Curve {keys []}}}
    MaskSpawnGroup=false
    Module.Enable Lathe Profile=false
    Module.Override Randomness=false
    Module.Surface Only=false
   ) {enabled=false, module=CylinderLocation001}
   &SampleSkeletalMeshSkeleton(Mesh:NiagaraDataInterfaceSkeletalMesh=$User.SkeletalMesh, Specific Bone Index:NiagaraInt32=$Particles.SpawnGroup) {enabled=false}
   &SkeletalMeshSkeletonLocation(Module.Inherit Velocity=false, Module.Use Sampled Bone as Ribbon ID=true) {enabled=false}
   &InheritVelocity(Module.Limit Speed=true, Velocity Limit=0.1, Velocity Scale=<-1 -1 -1>) {enabled=false}
   &AddVelocity(
    Scale Added Velocity=&Lerp_Vector(
     A=<0 0 0.3>
     B=<0 0 2>
     Alpha=&FloatFromCurve(
      FloatCurve=DataInterface() {propertyValues {Curve {keys [<at 0 0>,<at 1 1>]}}}
      CurveIndex=&Divide_Float(A=&VectorLength(Vector=$Particles.FallingSpeed), B=3000)
     )
    )
    Velocity=<0 0 1000>
   )
   &InitialMeshOrientation(
    Orientation Vector=&UniformRangedVector(Minimum=<-1 -1 -1>, Maximum=<1 1 1>)
    Use Orientation Vector=true
    Use Rotation Vector=false
   ) {enabled=false}
   /SetVariables(Particles.Color=&MakeLinearColorFromVectorAndFloat(
     "Vector (RGB)"=&MakeVectorFromLinearColorRGB(Color=&ColorFromCurve(ColorCurve=DataInterface() {propertyValues {
         AlphaCurve {keys [<at -0.0023 1 RCIM_Cubic>,<at 0.3756 1>,<at 0.9977 1 RCIM_Cubic>]}
         BlueCurve {keys [<at 0.3756 0.0782>,<at 0.4415 0.0786 RCIM_Cubic>]}
         GreenCurve {keys [<at 0.3756 0>,<at 0.4415 0 RCIM_Cubic>]}
         RedCurve {keys [<at 0.3756 1>,<at 0.4415 1 RCIM_Cubic>]}
        }}))
     "Float (Alpha)"=1
    )) {enabled=false, isSetParameters=true, module=SetVariables_#2}
   &InheritVelocity(Module.Limit Speed=false, Velocity Scale=<1 1 1.05>) {enabled=false, module=InheritVelocity001}
   &SphereLocation(MaskSpawnGroup=false, Module.Override Randomness=false, Non Uniform Scale=<1.5 1.5 1.5>, Sphere Radius=20, Surface Only=false) {enabled=false}
   &DynamicMaterialParameters(
    Index 0 Param 1=$Particles.Dy_offsetUV
    Index 0 Param 2=$Particles.Dy_U_Tile
    Index 0 Param 3=$Particles.Dy_Darkness
    Module.Param0WriteEnabled=true
    Module.Param1WriteEnabled=true
    Module.Param2WriteEnabled=true
    Module.Param3WriteEnabled=true
   )
  ]
  particleUpdateScript [
   &SkeletalMeshLocation(
    Bone / Socket Index=&UniformRangedInt(Minimum=4, Maximum=13)
    Bone Sampling Mode=::NewEnumerator4
    Inherit Velocity=false
    MaskSpawnGroup=false
    Sampled Position Offset=&UniformRangedVector(Minimum=<-3 -3 0>, Maximum=<3 3 1>)
    Skeletal Mesh=$User.SkeletalMesh
   ) {enabled=false}
   &ParticleState()
   &Drag(Drag=0.5, Use Linear Drag=true, Use Rotational Drag=false) {enabled=false}
   &CurlNoiseForce(
    Noise Frequency:NiagaraFloat=&UniformRangedFloat(Minimum=0.1, Maximum=1)
    Noise Strength=&UniformRangedFloat(Minimum=300, Maximum=600)
    Pan Noise=false
    Pan Noise Field=<1 1 0>
   ) {enabled=false}
   &SolveForcesAndVelocity(Clamp Velocity=false, Limit Acceleration=false)
   &ScaleSpriteSize(Scale Factor=&Multiply_Vector2(
     A=<1 1>
     B=&Vector2DFromFloat(Value=&FloatFromCurve(FloatCurve=DataInterface() {propertyValues {Curve {keys [
           <at 0 0.15 RCIM_Cubic>
           {arriveTangent=0.0667, leaveTangent=0.0667, time=0.1259, value=1}
           {arriveTangent=-1.1418, interpMode=RCIM_Cubic, leaveTangent=-1.1418, time=0.8397, value=1}
           {interpMode=RCIM_Cubic, tangentMode=RCTM_User, time=1, value=0.002}
          ]}}}))
    )) {enabled=false}
   &ScaleSpriteSize(Scale Factor=&Multiply_Vector2(
     A=<1 1>
     B=&Vector2DFromFloat(Value=&FloatFromCurve(
       FloatCurve=DataInterface() {propertyValues {Curve {keys [<at 0.4336 1.0007>,<at 0.4757 0.8507>,<at 0.5438 1.5807>,<at 0.6167 0.9007>,<at 0.6539 1.0007>,<at 1 1 RCIM_Cubic>]}}}
       Scale Curve=1.2
      ))
    )) {enabled=false, module=ScaleSpriteSize001}
   &ScaleColor(
    Scale Alpha=&FloatFromCurve(FloatCurve=DataInterface() {propertyValues {Curve {keys [<at 0 0 RCIM_Cubic>,<at 0.3634 1.0074 RCIM_Cubic>,<at 1 0 RCIM_Cubic>]}}})
    ScaleA=true
    ScaleRGB=true
   )
   &CameraOffset(Camera Offset Amount=10) {enabled=false}
  ]
  rendererClasses [&NiagaraSpriteRendererProperties]
  renderers [{rendererClass=&NiagaraSpriteRendererProperties}]
 }
 SpeedLine_Thigh {
  emitterSpawnScript [&FN_PreserveSkydiveUserVariables() {module=PreserveUserVariables}]
  emitterUpdateScript [
   &EmitterState(Life Cycle Mode=::NewEnumerator1, Loop Duration=3, Scalability Mode=::NewEnumerator1, UseLoopDelay=false)
   &SpawnBurst_Instantaneous(Spawn Count=2, Use Spawn Probability=false) {enabled=false}
   &SpawnRate(
    SpawnGroup:NiagaraInt32=&UniformRangedInt(Minimum=12, Maximum=17)
    SpawnRate=10
    Use Spawn Probability=false
   )
  ]
  particleSpawnScript [
   &FN_SetSpawnGroup(Emitter Bone Count=2) {enabled=false}
   /SetVariables(
    Particles.Dy_Darkness:NiagaraFloat=&UniformRangedFloat(Minimum=10, Maximum=50)
    Particles.Dy_U_Tile:NiagaraFloat=&UniformRangedFloat(Minimum=0.5, Maximum=1)
    Particles.Dy_offsetUV:NiagaraFloat=&UniformRangedFloat(Minimum=1, Maximum=2)
    Particles.FallingSpeed:Vector3f=$Engine.Owner.Velocity
    Particles.PositionRandom:Vector3f=&UniformRangedVector(Minimum=<-5 -10 0>, Maximum=<5 10 5>)
   ) {isSetParameters=true, module=SetVariables_#1}
   &InitializeParticle(
    Color=<rgba 1 0.1583 0.0084 1>
    Lifetime=&UniformRangedFloat(Minimum=1, Maximum=1)
    Mesh Scale=&Multiply_Vector(
     A=<1 1 1>
     B=&VectorFromFloat(Value=&UniformRangedFloat(Minimum=0.2, Maximum=1))
    )
    Module.Write Color=true
    Module.Write Lifetime=true
    Module.Write Mass=true
    Module.Write Material Random=true
    Module.Write Position=false
    Module.Write Scale=false
    Module.Write SpriteRotation=true
    Module.Write SpriteSize=true
    Position=&SimulationPosition()
    Sprite Rotation=90
    Sprite Size=&MakeVector2D(
     X=&Multiply_Float(A=10, B=&Lerp_Float(A=0.5, B=1.5, Alpha=$Particles.Dy_U_Tile))
     Y=&Multiply_Float(
      A=&UniformRangedFloat(Minimum=60, Maximum=150)
      B=&FloatFromCurve(
       FloatCurve=DataInterface() {propertyValues {Curve {keys [<at 0 1 RCIM_Cubic>, {arriveTangent=10.2223, interpMode=RCIM_Cubic, leaveTangent=10.2222, tangentMode=RCTM_User, time=1, value=4}]}}}
       CurveIndex=&Divide_Float(A=&VectorLength(Vector=$Particles.FallingSpeed), B=6000)
      )
     )
    )
   )
   &SampleSkeletalMeshSkeleton(Mesh:NiagaraDataInterfaceSkeletalMesh=$User.SkeletalMesh, Specific Bone Index:NiagaraInt32=$Particles.SpawnGroup) {enabled=false}
   &SkeletalMeshSkeletonLocation(Module.Inherit Velocity=false, Module.Use Sampled Bone as Ribbon ID=true) {enabled=false}
   &AddVelocity(Velocity=<0 0 1>)
   &DynamicMaterialParameters(
    Index 0 Param 1=$Particles.Dy_offsetUV
    Index 0 Param 2=$Particles.Dy_U_Tile
    Index 0 Param 3=$Particles.Dy_Darkness
    Index 0 Param 4=&UniformRangedFloat(Minimum=0, Maximum=0.05)
    Module.Param0WriteEnabled=true
    Module.Param1WriteEnabled=true
    Module.Param2WriteEnabled=true
    Module.Param3WriteEnabled=true
   )
   &CameraOffset(Camera Offset Amount=&UniformRangedFloat(Minimum=5, Maximum=10))
  ]
  particleUpdateScript [
   &SkeletalMeshLocation(
    Bone / Socket Index=&SpawnGroup()
    Bone Sampling Mode=::NewEnumerator4
    Inherit Velocity=false
    MaskSpawnGroup=false
    Sampled Position Offset=$Particles.PositionRandom
    Skeletal Mesh=$User.SkeletalMesh
   )
   &ParticleState()
   &SolveForcesAndVelocity(Clamp Velocity=false, Limit Acceleration=false)
   &ScaleColor(
    Scale Alpha=&FloatFromCurve(FloatCurve=DataInterface() {propertyValues {Curve {keys [<at 0 0 RCIM_Cubic>,<at 0.3634 1.0074 RCIM_Cubic>,<at 1 0 RCIM_Cubic>]}}})
    ScaleA=true
    ScaleRGB=true
   )
  ]
  rendererClasses [&NiagaraSpriteRendererProperties]
  renderers [{rendererClass=&NiagaraSpriteRendererProperties}]
 }
}
summary {
 systemName=NS_Contrail_ComicSpeedLine
 emitters [
  {emitterName=SpeedLine_Hands, rendererClasses [&NiagaraSpriteRendererProperties]}
  {bEnabled=false, emitterName=SpeedLine_Radial, rendererClasses [&NiagaraSpriteRendererProperties]}
  {emitterName=SpeedLine_Feet, rendererClasses [&NiagaraSpriteRendererProperties]}
  {bEnabled=false, emitterName=SpeedLine_Body, rendererClasses [&NiagaraSpriteRendererProperties]}
  {emitterName=SpeedLine_Surround, simTarget=GPUComputeSim, rendererClasses [&NiagaraSpriteRendererProperties]}
  {emitterName=SpeedLine_Shoulder, rendererClasses [&NiagaraSpriteRendererProperties]}
  {emitterName=SpeedLine_Thigh, rendererClasses [&NiagaraSpriteRendererProperties]}
 ]
 userVariables [
  User.bIsFrontEndPreview {type=NiagaraBool}
  User.FEVec {type=Vector3f, description="", defaultValue {struct=Vector3f, value=<0 0 1000>}}
  User.PlayerSpeed {type=NiagaraFloat, defaultValue=0, description=""}
  User.RibbonIDOffset {type=NiagaraInt32}
  User.SkeletalMesh {
   type=NiagaraDataInterfaceSkeletalMesh
   defaultValue {
    struct=NiagaraExt_VariableValue_DataInterface
    value {dataInterface=&self:NiagaraDataInterfaceSkeletalMesh_0}
   }
  }
  User.SpawnRate {type=NiagaraFloat, defaultValue=360, description=""}
 ]
}
